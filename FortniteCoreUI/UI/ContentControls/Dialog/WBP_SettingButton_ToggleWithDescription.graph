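# UIKit button widget for a toggleable setting row, with name, description, optional "New" tag and a disabled-message info box.
# PreConstruct registers the outline and toggle blocks with the modular button and tints the info box text yellow.
# Selection events drive the toggle block's state.
# SetDataStruct stores the option's display info, fills both texts, shows the tag for new options, and disables the button and reveals the info box for disabled ones.
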WidgetBlueprint /FortniteCoreUI/UI/ContentControls/Dialog/WBP_SettingButton_ToggleWithDescription
graphs [
 EventGraph {nodes [
   n1 = Event PreConstruct()
   > Parent: PreConstruct(IsDesignTime=@n1.IsDesignTime)
   > SetDefaultColorAndOpacity!(
    execute
    self:Rich Text Block Object Reference=Get Text(self:WBP UIKit Info Box Object Reference=Get WBP_UIKit_InfoBox_DisabledMessage() -> (WBP_UIKit_InfoBox_DisabledMessage:WBP UIKit Info Box Object Reference)) -> (Text:Common Rich Text Block Object Reference)
    InColorAndOpacity:Slate Color Structure=Make Slate Color(SpecifiedColor=(R=0.9301,G=1,B=0.0103,A=1))
   ) -> (then)
   > RegisterModularBlock!(
    execute
    self:UIKit Modular Button Object Reference
    Block:UIKit Block Object Reference=Get Block_Outline() -> (Block_Outline:WBP UIKit Block Outline Object Reference)
   ) -> (then)
   > RegisterModularBlock(Block=Get Block_Toggle() -> (Block_Toggle:WBP UIKit Block Toggle Object Reference))
   Event On Selected()
   > Parent: On Selected() {is=title K2Node_CallParentFunction}
   > SetSelected!(execute, self:WBP UIKit Block Toggle Object Reference=Get Block_Toggle(), Selected:bool=true) -> (then)
   Event On Deselected()
   > Parent: On Deselected() {is=title K2Node_CallParentFunction}
   > SetSelected(self=Get Block_Toggle(), Selected=false)
  ]}
 SetDataStruct {nodes [
   n1 = SetDataStruct!() -> (then, ItemData:Selectable Setting Option Display Info Structure) {is=title K2Node_FunctionEntry}
   > n2 = Set DataStruct(DataStruct:Selectable Setting Option Display Info Structure=@n1) -> (Output_Get:Selectable Setting Option Display Info Structure)
   n3 = Break Selectable Setting Option Display Info!(SelectableSettingOptionDisplayInfo:"Selectable Setting Option Display Info Structure (by ref)"=@n2) -> (OptionId:String, DisplayName:Text, Description:Text, bIsEnabled:bool=~, bIsNew:bool=~) {is=title K2Node_BreakStruct}
   > "SetText (Text)"(self=Get SettingName_Text() -> (SettingName_Text:Common Text Object Reference), InText=@n3.DisplayName)
   > "SetText (Text)"(self=Get DescriptionText() -> (DescriptionText:Common Text Object Reference), InText=@n3.Description)
   > SetVisibility(self=Get WBP_UIKit_InfoBox_DisabledMessage(), InVisibility=Select(Option 0=SelfHitTestInvisible, Option 1=Collapsed, Index=@n3.bIsEnabled))
   > SetIsEnabled(self=Self-Reference(), bInIsEnabled=@n3.bIsEnabled)
   > SetVisibility(
    self=Get WBP_UIKit_Tag_NEW() -> (WBP_UIKit_Tag_NEW:WBP UIKit Tag Object Reference)
    InVisibility=Select(Option 0=Collapsed, Option 1=SelfHitTestInvisible, Index=@n3.bIsNew)
   )
  ]}
]
variables [DataStruct {type=SelectableSettingOptionDisplayInfo, default {bIsEnabled=false, bIsNew=false, description="", displayName="", optionId=""}}]
widgets {
 info {
  inheritedWidgetCount=0
  namedSlotCount=1
  parentClass=/FortUILibrary/UIKit/Buttons/WBP_UIKit_Button_Generic.WBP_UIKit_Button_Generic_C
  rootWidgetClass=null
  widgetCount=10
 }
 widgets [
  Overlay_0 {widgetClassPath=&Overlay, widget=Overlay_0}
  ToggleOverlay {widgetClassPath=&Overlay, parent=Overlay_0, widget=ToggleOverlay, slot {class=OverlaySlot, properties {padding=<ltrb 0 0 30 0>}}}
  Block_Outline {
   widgetClassPath=/FortUILibrary/UIKit/Buttons/Blocks/WBP_UIKit_Block_Outline
   bIsVariable=true
   parent=ToggleOverlay
   widget=Block_Outline
   properties {bottomLeftCornerRadius=1, bottomRightCornerRadius=1, topLeftCornerRadius=1, topRightCornerRadius=1}
   slot {class=OverlaySlot, properties {horizontalAlignment=HAlign_Fill, verticalAlignment=VAlign_Fill}}
  }
  Block_Toggle {
   widgetClassPath=/FortUILibrary/Systems/ModularButton/Blocks/WBP_UIKit_Block_Toggle
   bIsVariable=true
   parent=ToggleOverlay
   widget=Block_Toggle
   slot {class=OverlaySlot, properties {horizontalAlignment=HAlign_Fill, verticalAlignment=VAlign_Fill}}
  }
  VerticalBox_1 {
   widgetClassPath=&VerticalBox
   parent=Overlay_0
   widget=VerticalBox_1
   slot {class=OverlaySlot, properties {horizontalAlignment=HAlign_Fill, padding=<ltrb 64 0 0 0>}}
  }
  HorizBox_SettingName {
   widgetClassPath=&HorizontalBox
   parent=VerticalBox_1
   widget=HorizBox_SettingName
   slot {class=VerticalBoxSlot, properties {verticalAlignment=VAlign_Top}}
  }
  SettingName_Text {
   widgetClassPath=&CommonTextBlock
   bIsVariable=true
   parent=HorizBox_SettingName
   widget=SettingName_Text
   properties {
    text=Setting name
    font {
     bForceMonospaced=false
     bMaterialIsStencil=false
     fontMaterial=/FortUILibrary/UIKit/Buttons/MaterialInstances/GenericButtons/MI_UIKit_ButtonWithToggle_Text
     fontObject=&HeadingNow
     letterSpacing=0
     monospacedWidth=1
     size=18
     skewAmount=0
     typefaceFontName=75 Medium
     outlineSettings {
      bApplyOutlineToDropShadows=false
      bMiteredCorners=false
      bSeparateFillAlpha=false
      outlineMaterial=null
      outlineSize=0
      outlineColor=<rgba 0 0 0 1>
     }
    }
    shadowOffset=<0 0>
   }
   slot {class=HorizontalBoxSlot, properties {horizontalAlignment=HAlign_Left}}
  }
  WBP_UIKit_Tag_NEW {
   widgetClassPath=/FortUILibrary/UIKit/Tag/WBP_UIKit_Tag
   bIsVariable=true
   parent=HorizBox_SettingName
   widget=WBP_UIKit_Tag_NEW
   properties {labelText="", tagSize=Compact, tagText=New, tagType=Primary Tag, colorAndOpacity=<rgba 0.9301 1 0.0103 1>}
   slot {class=HorizontalBoxSlot, properties {padding=<ltrb 12 0 0 0>}}
  }
  DescriptionText {
   widgetClassPath=&CommonTextBlock
   bIsVariable=true
   parent=VerticalBox_1
   widget=DescriptionText
   properties {
    autoWrapText=true
    text=Setting description lorem ipsum dolor sit amet extra long might wrap more text even more
    font {
     bForceMonospaced=false
     bMaterialIsStencil=false
     fontMaterial=&MI_UIKit_ButtonWithToggle_Text
     fontObject=&HeadingNow
     letterSpacing=0
     monospacedWidth=1
     size=18
     skewAmount=0
     typefaceFontName=74 Regular
     outlineSettings {
      bApplyOutlineToDropShadows=false
      bMiteredCorners=false
      bSeparateFillAlpha=false
      outlineMaterial=null
      outlineSize=0
      outlineColor=<rgba 0 0 0 1>
     }
    }
    shadowOffset=<0 0>
   }
   slot {class=VerticalBoxSlot}
  }
  WBP_UIKit_InfoBox_DisabledMessage {
   widgetClassPath=/FortUILibrary/UIKit/InfoBox/WBP_UIKit_InfoBox
   bIsVariable=true
   parent=VerticalBox_1
   widget=WBP_UIKit_InfoBox_DisabledMessage
   properties {
    iconMaterial=/FortUILibrary/Textures/MaterialInstances/MI_UIKit_Icon_AgeGatedAlert
    previewText=Unavailable for children under 13
    show Background=false
    use Preview Text=true
   }
   slot {class=VerticalBoxSlot, properties {padding=<ltrb 0 8 0 0>}}
  }
 ]
}
