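# Parental-controls migration dialog asking a parent to choose a replacement for a deprecated child chat setting.
# Builds radio-style tabs from setting options; selection enables the confirm button, which submits through FortDeprecatedSupervisedSettingModalVM and closes, while the second opens a learn-more URL.
# On gamepad, a no-op button takes focus to scroll the description into view.
# Activation and deactivation broadcast modal-active events unless the code-driven focus-hack runtime option is on.
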WidgetBlueprint /FortniteCoreUI/UI/ContentControls/Dialog/WBP_Dialog_MigrationChatSettings
graphs [
 BP_GetDesiredFocusTarget {nodes [
   Get Desired Focus Target()
   > n1 = Branch(Condition=GetIsEnabled!(self:Widget Object Reference=Get Button_ChooseSettingDoesntDoAnything() -> (Button_ChooseSettingDoesntDoAnything:WBP UIKit Button Generic Object Reference)) -> (ReturnValue:bool))
   > Return Node!(execute, ReturnValue:Widget Object Reference=Get Button_ChooseSettingDoesntDoAnything())
   n2 = For Each Loop(
    Exec=@n1.else
    Array:Array of Selectable Setting Option Display Info Structures=Get SettingOptionsDisplayInfo!(self:Fort Deprecated Supervised Setting Modal VM Object Reference=Get FortDeprecatedSupervisedSettingModalVM() -> (FortDeprecatedSupervisedSettingModalVM:Fort Deprecated Supervised Setting Modal VM Object Reference)) -> (SettingOptionsDisplayInfo:Array of Selectable Setting Option Display Info Structures)
   ) -> (Array Element:Selectable Setting Option Display Info Structure)
   > Branch(Condition=Break Selectable Setting Option Display Info!(SelectableSettingOptionDisplayInfo:"Selectable Setting Option Display Info Structure (by ref)"=@n2) -> (OptionId:String, DisplayName:Text, Description:Text, bIsEnabled:bool=~, bIsNew:bool=~) {is=title K2Node_BreakStruct}.bIsEnabled)
   n3 = Get WBP_UIKit_TabGroup_RadioSelection() -> (WBP_UIKit_TabGroup_RadioSelection:WBP UIKit Tab Group Object Reference)
   > Return Node(ReturnValue=GetTabButtonBaseByID!(
     self:Common Tab List Widget Base Object Reference=@n3
     TabNameID:Name=GetTabIdAtIndex!(self:Common Tab List Widget Base Object Reference=@n3, Index:int=@n2.Array Index) -> (ReturnValue:Name=None)
    ) -> (ReturnValue:Common Button Base Object Reference))
   n4 = Get WBP_UIKit_TabGroup_RadioSelection()
   Return Node(execute=@n2.Completed, ReturnValue=GetTabButtonBaseByID(self=@n4, TabNameID=GetTabIdAtIndex(self=@n4, Index=0) -> (ReturnValue=None)))
  ]}
 CF_DescriptionText {nodes [
   n1 = CF_DescriptionText!() -> (then, CurrentSettingState:EDeprecatedSettingState Enum=None) {is=title K2Node_FunctionEntry}
   > Return Node!(
    execute
    OutDescriptionText:Text=Select!(
     None:Text="LOCTEXT(\"<Body10>We’ve added a new chat setting:</><Body10_Strong> Friends, team members, and connections.</> \\r\\n\\r\\n<Body10>Connections are people who are not your child’s direct friends but are linked to your child through their friends. Your current setting,</> <Body10_Strong>Friends and teammates</><Body10>, will apply until you update it.</>\")"
     VoiceOnly:Text="LOCTEXT(\"<Body10>We’ve added a new chat setting:</><Body10_Strong> Friends, team members, and connections.</> \\r\\n\\r\\n<Body10>Connections are people who are not your child’s direct friends but are linked to your child through their friends. Your current setting,</> <Body10_Strong>Friends and teammates</><Body10>, will apply until you update it.</>\")"
     TextOnly:Text="LOCTEXT(\"<Body10>We’ve added a new chat setting:</><Body10_Strong> Friends, team members, and connections.</> \\r\\n\\r\\n<Body10>Connections are people who are not your child’s direct friends but are linked to your child through their friends. Your current setting,</> <Body10_Strong>Friends and teammates</><Body10>, will apply until you update it.</>\")"
     VoiceAndText:Text="LOCTEXT(\"<Body10>We’ve added a new chat setting:</><Body10_Strong> Friends, team members, and connections.</> \\r\\n\\r\\n<Body10>Connections are people who are not your child’s direct friends but are linked to your child through their friends. Your current setting,</> <Body10_Strong>Friends and teammates</><Body10>, will apply until you update it.</>\")"
     Index:EDeprecatedSettingState Enum=@n1
    ) -> (ReturnValue:Text)
   )
  ]}
 CreateRadioTabGroup {nodes [
   n1 = CreateRadioTabGroup!() -> (then, SelectableSettingsArray:Array of Selectable Setting Option Display Info Structures) {is=title K2Node_FunctionEntry}
   > RemoveAllTabs(self:Common Tab List Widget Base Object Reference=Get WBP_UIKit_TabGroup_RadioSelection())
   > Branch(Condition=Is Not Empty!(TargetArray:Array of Selectable Setting Option Display Info Structures=@n1) -> (ReturnValue:bool))
   > n2 = For Each Loop(Array=Get SelectableSettingsArray() -> (SelectableSettingsArray:Array of Selectable Setting Option Display Info Structures))
   n3 = String To Name(InString=Break Selectable Setting Option Display Info(SelectableSettingOptionDisplayInfo=@n2) -> (bIsEnabled=~, bIsNew=~))
   > AddTab!(
    execute
    self:WBP UIKit Tab Group Base Object Reference=Get WBP_UIKit_TabGroup_RadioSelection()
    Tab VM:MVVM Base Viewmodel Object Reference
    Tab ID:Name=@n3
    Index:int=@n2.Array Index
   ) -> (then)
   > n4 = Cast To WBP_SettingButton_ToggleWithDescription!(execute, Object=GetTabButtonBaseByID(self=Get WBP_UIKit_TabGroup_RadioSelection(), TabNameID=@n3)) -> (then, CastFailed, AsWBP Setting Button Toggle with Description:WBP Setting Button Toggle With Description Object Reference) {is=title K2Node_DynamicCast}
   > SetDataStruct!(execute, self:WBP Setting Button Toggle With Description Object Reference=@n4, ItemData:Selectable Setting Option Display Info Structure=@n2) -> (then)
  ]}
 EventGraph {nodes [
   "On Focused (Button_ChooseSettingDoesntDoAnything)"!() -> (OutputDelegate, then, Button:Common Button Base Object Reference) {is=title K2Node_ComponentBoundEvent}
   > ScrollWidgetIntoView!(
    execute
    self:Scroll Box Object Reference=Get ScrollBox_0() -> (ScrollBox_0:Scroll Box Object Reference)
    WidgetToFind:Widget Object Reference=Get Description_RichText() -> (Description_RichText:Common Rich Text Block Object Reference)
    AnimateScroll:bool=true
    ScrollDestination:EDescendantScrollDestination Enum=IntoView
    Padding:float=0
   ) -> (then)
   "On Tab Selected (WBP_UIKit_TabGroup_RadioSelection)"!() -> (OutputDelegate, then, TabId:Name=None) {is=title K2Node_ComponentBoundEvent}
   > SetButtonEnabled!(
    execute
    self:WBP UIKit Dialog Internal Object Reference=Get WBP_UIKit_Dialog_Internal() -> (WBP_UIKit_Dialog_Internal:WBP UIKit Dialog Internal Object Reference)
    ButtonIndex:int=0
    IsButtonEnabled:bool=true
   ) -> (then)
   n1 = OnDialogButtonClicked!() -> (OutputDelegate, then, ButtonIndex:int) {is=title K2Node_CustomEvent}
   > n2 = Switch on Int!(execute:Exec, Selection:int=@n1.ButtonIndex) -> (0:Exec, 1:Exec) {bHasDefaultPin=false}
   n3 = Get WBP_UIKit_TabGroup_RadioSelection()
   > SubmitSettingOptionSelection!(
    execute
    self:Fort Deprecated Supervised Setting Modal VM Object Reference=Get FortDeprecatedSupervisedSettingModalVM()
    SettingId:String="To String (Name)"(InName=GetSelectedTabId!(self:Common Tab List Widget Base Object Reference=@n3) -> (ReturnValue:Name=None))
   ) -> (then)
   > DeactivateWidget()
   n4 = GetLearnMoreURLString!(self:Fort Deprecated Supervised Setting Modal VM Object Reference=Get FortDeprecatedSupervisedSettingModalVM()) -> (ReturnValue:String)
   LaunchFortURL!(
    execute=@n2.1
    Url:String=SelectString(A=https://safety.epicgames.com/en-US/parental-controls, B=@n4, bPickA=IsEmpty(InString=@n4))
    bAddModeParam:bool=true
    bAddLanguageParam:bool=true
   ) -> (then, ReturnValue:bool)
   n5 = OnHandleInputMethodChanged!() -> (OutputDelegate, then, NewInputType:ECommonInputType Enum) {is=title K2Node_CustomEvent}
   > SetScrollWhenFocusChanges!(
    execute
    self:Scroll Box Object Reference=Get ScrollBox_0()
    NewScrollWhenFocusChanges:EScrollWhenFocusChanges Enum=Select!(
     MouseAndKeyboard:EScrollWhenFocusChanges Enum=NoScroll
     Gamepad:EScrollWhenFocusChanges Enum=AnimatedScroll
     Touch:EScrollWhenFocusChanges Enum=NoScroll
     Count:EScrollWhenFocusChanges Enum=NoScroll
     Index:ECommonInputType Enum=@n5.NewInputType
    ) -> (ReturnValue:EScrollWhenFocusChanges Enum)
   ) -> (then)
   > SetIsEnabled(
    self=Get Button_ChooseSettingDoesntDoAnything()
    bInIsEnabled=Select(MouseAndKeyboard:bool=false, Gamepad:bool=true, Touch:bool=false, Count:bool=false, Index=@n5.NewInputType) -> (ReturnValue:bool)
   )
   Event Destruct()
   n6 = Create Event()
   n7 = Get CommonInputSubsystem()
   > Unbind Event from On Input Method Changed(self=@n7, Delegate=@n6)
   n8 = Create Event()
   n9 = Get WBP_UIKit_Dialog_Internal()
   > Unbind Event from Dialog Button Clicked Dispatcher(self:WBP UIKit Dialog Internal Object Reference=@n9, Delegate=@n8) {is=title K2Node_RemoveDelegate}
   Event On Activated()
   n10 = Get bDeprecatedSettingModalUseCodeDrivenFocusHack!(self:Fort Runtime Options Object Reference=GetRuntimeOptions()) -> (bDeprecatedSettingModalUseCodeDrivenFocusHack:bool)
   > n11 = Branch(Condition=@n10)
   n12 = GetOwningPlayer()
   n13 = GetGlobalEventRouter()
   Broadcast Event!(
    execute=@n11.else
    self:Gameplay Event Router Component Object Reference=@n13
    EventData:BPS Parental Controls Migration Settings Modal Active Change Structure=Make BPS Parental Controls Migration Settings Modal Active Change!(bActive_4_#1:bool=true) -> (BPS_ParentalControlsMigrationSettingsModalActiveChange:BPS Parental Controls Migration Settings Modal Active Change Structure) {is=title K2Node_MakeStruct}
    EventContext:Object Reference=@n12
   ) -> (then, ReturnValue:bool)
   Event On Deactivated()
   > n14 = Branch(Condition=@n10)
   Broadcast Event(
    execute=@n14.else
    self=@n13
    EventData=Make BPS Parental Controls Migration Settings Modal Active Change(bActive_4_#1=False)
    EventContext=@n12
   )
   Event Construct()
   > Bind Event to On Input Method Changed(self=@n7, Delegate=@n6)
   > OnHandleInputMethodChanged!(execute, self, NewInputType:ECommonInputType Enum=GetCurrentInputType(self=@n7)) -> (then) {is=title K2Node_CallFunction}
   > n15 = Construct Object from Class(Class=/FortniteCoreUI/UI/ContentControls/Dialog/VM/DialogVM_MigrationChatSettings.DialogVM_MigrationChatSettings_C) -> (ReturnValue:Dialog VM Migration Chat Settings Object Reference)
   > Set DialogTitle!(
    execute
    DialogTitle:Text=GetModalTitleText!(self:Fort Deprecated Supervised Setting Modal VM Object Reference=Get FortDeprecatedSupervisedSettingModalVM()) -> (ReturnValue:"Text (by ref)")
    self:UI Kit Dialog Viewmodel Object Reference=@n15
   ) -> (then, Output_Get:Text)
   > SetDialogData!(execute, self:WBP UIKit Dialog Internal Object Reference=Get WBP_UIKit_Dialog_Internal(), Data:UI Kit Dialog Viewmodel Object Reference=@n15) -> (then)
   > Bind Event to Dialog Button Clicked Dispatcher(self:WBP UIKit Dialog Internal Object Reference=@n9, Delegate=@n8) {is=title K2Node_AddDelegate}
   > SetScrollWidget!(execute, ToScroll:Widget Object Reference=Get ScrollBox_0()) -> (then)
   > SetButtonEnabled(self=Get WBP_UIKit_Dialog_Internal(), ButtonIndex=0, IsButtonEnabled=false)
  ]}
 OnNavDownFromRadioPickButton {nodes [
   OnNavDownFromRadioPickButton!() -> (then, Navigation:EUINavigation Enum) {is=title K2Node_FunctionEntry}
   n1 = Get ButtonWidgets!(self:WBP UIKit Dialog Internal Object Reference=Get WBP_UIKit_Dialog_Internal()) -> (ButtonWidgets:Array of WBP UIKit Button CTA Base Object References)
   n2 = "Get (a copy)"(Array:Array of WBP UIKit Button CTA Base Object References=@n1) -> (Output:WBP UIKit Button CTA Base Object Reference)
   > n3 = Branch(Condition=GetIsEnabled(self=@n2))
   > Return Node!(execute, ReturnValue:Widget Object Reference=@n2)
   Return Node(execute=@n3.else, ReturnValue="Get (a copy)"(Array=@n1, Dimension 1=1))
  ]}
]
variables []
widgets {
 info {
  inheritedWidgetCount=0
  namedSlotCount=1
  parentClass=&CommonActivatableWidget
  rootWidgetClass=/FortUILibrary/Systems/Dialog/WBP_UIKit_Dialog_Internal.WBP_UIKit_Dialog_Internal_C
  widgetCount=7
 }
 widgets [
  WBP_UIKit_Dialog_Internal {
   widgetClassPath=/FortUILibrary/Systems/Dialog/WBP_UIKit_Dialog_Internal
   bIsVariable=true
   widget=WBP_UIKit_Dialog_Internal
   properties {
    designTimePreview_DialogVM=&DialogVM_MigrationChatSettings_C
    dialogSIze {
     height Override {enable=true, value=1000}
     max Aspect Ratio {enable=false, value=0}
     max Desired Height {enable=true, value=800}
     max Desired Width {enable=false, value=0}
     min Aspect Ratio {enable=false, value=0}
     min Desired Height {enable=false, value=0}
     min Desired Width {enable=false, value=0}
     width Override {enable=true, value=1300}
    }
   }
  }
  Overlay_0 {widgetClassPath=&Overlay, namedSlotHost=WBP_UIKit_Dialog_Internal, widget=Overlay_0}
  ScrollBox_0 {
   widgetClassPath=&ScrollBox
   bIsVariable=true
   parent=Overlay_0
   widget=ScrollBox_0
   properties {scrollWhenFocusChanges=AnimatedScroll}
   slot {class=OverlaySlot, properties {padding=<ltrb 80 0 80 0>}}
  }
  VerticalBox_0 {
   widgetClassPath=&VerticalBox
   parent=ScrollBox_0
   widget=VerticalBox_0
   slot {class=ScrollBoxSlot, properties {padding=<ltrb 12 12 0 0>}}
  }
  Description_RichText {
   widgetClassPath=&CommonRichTextBlock
   bIsVariable=true
   parent=VerticalBox_0
   widget=Description_RichText
   properties {
    autoWrapText=true
    text="<Body10>We’ve added a new chat setting:</><Body10_Strong> Friends, team members, and connections.</> \r\n\r\n<Body10>Connections are people who are not your child’s direct friends but are linked to your child through their friends. Your current setting,</> <Body10_Strong>Friends and teammates</><Body10>, will apply until you update it.</>"
    textStyleSet=/FortUILibrary/Styles/DT_UI_RichTextStyles_HeadingNow
    defaultTextStyleOverride {
     overflowPolicy=Clip
     transformPolicy=None
     colorAndOpacity {colorUseRule=UseColor_Specified, specifiedColor=<rgba 1 1 1 1>}
     font {
      bForceMonospaced=false
      bMaterialIsStencil=false
      fontMaterial=null
      fontObject=&HeadingNow
      letterSpacing=0
      monospacedWidth=1
      size=16.5
      skewAmount=0
      typefaceFontName=74 Regular
      outlineSettings {
       bApplyOutlineToDropShadows=false
       bMiteredCorners=false
       bSeparateFillAlpha=false
       outlineMaterial=null
       outlineSize=0
       outlineColor=<rgba 0 0 0 1>
      }
     }
     highlightColor {colorUseRule=UseColor_Specified, specifiedColor=<rgba 0 0 0 1>}
     highlightShape {
      bIsDynamicallyLoaded=false
      drawAs=Image
      imageType=NoImage
      mirroring=NoMirror
      resourceName=None
      resourceObject=null
      tiling=NoTile
      imageSize=<32 32>
      margin=<ltrb 0 0 0 0>
      outlineSettings {
       bUseBrushTransparency=false
       roundingType=HalfHeightRadius
       width=0
       color {colorUseRule=UseColor_Specified, specifiedColor=<rgba 0 0 0 0>}
       cornerRadii=<0 0 0 0>
      }
      tintColor {colorUseRule=UseColor_Specified, specifiedColor=<rgba 1 1 1 1>}
      uVRegion {bIsValid=false, max=<0 0>, min=<0 0>}
     }
     selectedBackgroundColor {colorUseRule=UseColor_Foreground, specifiedColor=<rgba 1 0 1 1>}
     shadowColorAndOpacity=<rgba 0 0 0 1>
     shadowOffset=<0 0>
     strikeBrush {
      bIsDynamicallyLoaded=false
      drawAs=Image
      imageType=NoImage
      mirroring=NoMirror
      resourceName=None
      resourceObject=null
      tiling=NoTile
      imageSize=<32 32>
      margin=<ltrb 0 0 0 0>
      outlineSettings {
       bUseBrushTransparency=false
       roundingType=HalfHeightRadius
       width=0
       color {colorUseRule=UseColor_Specified, specifiedColor=<rgba 0 0 0 0>}
       cornerRadii=<0 0 0 0>
      }
      tintColor {colorUseRule=UseColor_Specified, specifiedColor=<rgba 1 1 1 1>}
      uVRegion {bIsValid=false, max=<0 0>, min=<0 0>}
     }
     underlineBrush {
      bIsDynamicallyLoaded=false
      drawAs=Image
      imageType=NoImage
      mirroring=NoMirror
      resourceName=None
      resourceObject=null
      tiling=NoTile
      imageSize=<32 32>
      margin=<ltrb 0 0 0 0>
      outlineSettings {
       bUseBrushTransparency=false
       roundingType=HalfHeightRadius
       width=0
       color {colorUseRule=UseColor_Specified, specifiedColor=<rgba 0 0 0 0>}
       cornerRadii=<0 0 0 0>
      }
      tintColor {colorUseRule=UseColor_Specified, specifiedColor=<rgba 1 1 1 1>}
      uVRegion {bIsValid=false, max=<0 0>, min=<0 0>}
     }
    }
   }
   slot {class=VerticalBoxSlot}
  }
  Button_ChooseSettingDoesntDoAnything {
   widgetClassPath=&WBP_UIKit_Button_Generic
   bIsVariable=true
   parent=VerticalBox_0
   widget=Button_ChooseSettingDoesntDoAnything
   properties {
    bHideInputAction=true
    backgroundMaterial=/FortUILibrary/UIKit/Buttons/MaterialInstances/CTAButtons/MI_UIKit_EmptyProgress
    minHeightOverride=54
    mobileMinHeight=80
    text=Choose a new setting to continue:
    textTransformPolicy=None
    font {
     bForceMonospaced=false
     bMaterialIsStencil=false
     fontMaterial=null
     fontObject=&HeadingNow
     letterSpacing=10
     monospacedWidth=1
     size=18
     skewAmount=0
     typefaceFontName=75 Medium
     outlineSettings {
      bApplyOutlineToDropShadows=false
      bMiteredCorners=false
      bSeparateFillAlpha=false
      outlineMaterial=/FortUILibrary/Systems/ModularButton/MaterialInstances/MI_UIKit_TextOutline
      outlineSize=0
      outlineColor=<rgba 0 0 0 1>
     }
    }
    padding=<ltrb 0 12 0 12>
   }
   slot {class=VerticalBoxSlot, properties {horizontalAlignment=HAlign_Left, verticalAlignment=VAlign_Top}}
  }
  WBP_UIKit_TabGroup_RadioSelection {
   widgetClassPath=/FortUILibrary/UIKit/Tabs/WBP_UIKit_TabGroup
   bIsVariable=true
   parent=VerticalBox_0
   widget=WBP_UIKit_TabGroup_RadioSelection
   properties {
    bIsFocusable=true
    buttonClass=&/WBP_SettingButton_ToggleWithDescription.WBP_SettingButton_ToggleWithDescription_C
    buttonHorizontalAlignment=HAlign_Fill
    buttonVerticalAlignment=VAlign_Top
    hidePrompts=true
    isNaturalNavigation=true
    isVertical=true
    navigation=&self:WidgetTree.WBP_UIKit_TabGroup_RadioSelection.WidgetNavigation_0
    tabGroupHorizontalAlignment=HAlign_Fill
    useBackplate=false
    usePromptBackplate=false
    initialTabsData [
     {
      bEnableTabId=false
      tabId=None
      tabViewModel=&self:WidgetTree.WBP_UIKit_TabGroup_RadioSelection.UIKitSelectableButtonViewModel_0
     }
     {
      bEnableTabId=false
      tabId=None
      tabViewModel=&self:WidgetTree.WBP_UIKit_TabGroup_RadioSelection.UIKitSelectableButtonViewModel_2
     }
     {
      bEnableTabId=false
      tabId=None
      tabViewModel=&self:WidgetTree.WBP_UIKit_TabGroup_RadioSelection.UIKitSelectableButtonViewModel_1
     }
    ]
   }
   slot {class=VerticalBoxSlot}
  }
 ]
}
